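; Fortnite Creative Volume device that broadcasts enter and exit events for players, vehicle drivers, creatures, guards and physics props.
; It builds a box, cylinder, sphere or custom-mesh volume and switches its material by play mode and visibility.
; Minigame start binds overlaps on the authority, including any linked external volume, and replays enters for pawns already inside.
; An overlap filter applies options and team/class checks, retrying unresolved player pawns on possession.
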(event EventInitializeDevice
  ;[0 1 0.7004 0.25] Set Volume Shape and Initialize Creative Use Restriction Component
  (|CallParentFunction)
  (CallFunction|InitializeCreativeIcon)
  (CallFunction|UpdateVisibilityComponent)
  (CallFunction|UpdateVolumeShape)
  (CallFunction|CheckVolumetoCache)
  (Initialization|Initialize (Variables|Default|GetCreative_UseRestriction_Component) (Variables|Default|GetFortMinigameProgress) 0 0 (Variables|UserOptions|GetSelectedTeam) (Variables|UserOptions|GetInvert%20Team%20Selection) (Variables|UserOptions|GetSelectedClass) (Variables|UserOptions|GetInvert%20Class%20Selection) false)
  ; Disable volume visibility trace collision on edit mode to handle phone manipulations
  (Collision|SetCollisionResponsetoChannel (Variables|Default|GetCollision) "ECC_Visibility" (select (CallFunction|IsinPlayMode) "ECR_Overlap" "ECR_Ignore")))

(event EditorPreview|EventUpdateEditorPreview
  (params (bHasMoved :type Boolean))
  ;[1 1 1] On Moved In Editor Event
  (|CallParentFunction bHasMoved)
  (if bHasMoved
    (Default|CallOnMovedinEditor self)))

(event EditorPreview|EventSetupForEditorPreview
  ;[1 1 1] On Moved In Editor Event
  (|CallParentFunction)
  (Default|CallOnMovedinEditor self))

(event PawnNowHasValidParticipantController(Creative_UseRestriction_Component)
  (params (FortPlayerPawn :type Fort%20Player%20Pawn%20Object%20Reference) (Controller :type Controller%20Object%20Reference))
  ;[0 1 0.8813 0.25] Make sure players spawning in volume have their controller to send enter events on spawn.
  (bind _involume (CallFunction|IsActorinVolume FortPlayerPawn))
  (if (and _involume (CreativeUserRestrictions|CanParticipantControllerInteract (Variables|Default|GetCreative_UseRestriction_Component) Controller (Variables|UserOptions|Advanced|GetGuardEventsEnabled) true false))
    (CallFunction|CachedVolumeOverlapStart FortPlayerPawn)))

(event Custom|CachedVolumeOverlapStart
  (params (OtherActor :type Actor%20Object%20Reference))
  ;[0 1 0.4541 0.25] Enter/ExitEvents
  (CallFunction|NormalorShippingLog (Utilities|String|Append "Volume Device: Overlap Has begun for " (Utilities|GetDisplayName self)))
  (CallFunction|FilterActorforEvent OtherActor))

(event MinigameStartedDelegate(FortMinigameProgress)
  (bind _outputdelegate (EventDispatchers|CreateEvent :!event Custom|CachedVolumeOverlapStart))
  (bind _outputdelegate_2 (EventDispatchers|CreateEvent :!event Custom|CachedVolumeOverlapEnd))
  (bind _cachedoverlapcomponent (Variables|Default|GetCachedOverlapComponent))
  (bind _fne_volumeoverlapcomponent (Variables|Default|GetFNE_VolumeOverlapComponent))
  ;[0 1 0.6126 0.25] Setting Event Bindings
  (Networking|SwitchHasAuthority
    (:Authority
      (Class|FNEVolumeOverlapComponent|BindtoonOtherActorBeginOverlap _cachedoverlapcomponent _outputdelegate)
      (Class|FNEVolumeOverlapComponent|BindtoonOtherActorEndOverlap _cachedoverlapcomponent _outputdelegate_2)
      (if (Variables|Default|GetHasExternalVolume)
        (Class|FNEVolumeOverlapComponent|BindtoonOtherActorBeginOverlap _fne_volumeoverlapcomponent _outputdelegate)
        (Class|FNEVolumeOverlapComponent|BindtoonOtherActorEndOverlap _fne_volumeoverlapcomponent _outputdelegate_2)
        ;;[0 1 0.7373 0.25] Trigger OnEnter events for valid players on game start.
        (for _array_element (Collision|GetOverlappingActors _fne_volumeoverlapcomponent "/Script/FortniteGame.FortPawn" :!class /Script/Engine.PrimitiveComponent)
          (Utilities|ValidTeamandClass _array_element
            (:Valid
              (CallFunction|CachedVolumeOverlapStart _array_element))))
        (else
          _))
      (for _array_element_2 (Collision|GetOverlappingActors _cachedoverlapcomponent "/Script/FortniteGame.FortPawn" :!class /Script/Engine.PrimitiveComponent)
        (Utilities|ValidTeamandClass _array_element_2
          (:Valid
            (CallFunction|CachedVolumeOverlapStart _array_element_2)))))))

(event Custom|CachedVolumeOverlapEnd
  (params (OtherActor :type Actor%20Object%20Reference))
  ;[0 1 0.4541 0.25] Enter/ExitEvents
  (CallFunction|NormalorShippingLog (Utilities|String|Append "Volume Device: Overlap Has ended for " (Utilities|GetDisplayName self)))
  (CallFunction|FilterActorforEvent OtherActor true))

(event MinigameEndedDelegate(FortMinigameProgress)
  (bind _outputdelegate (EventDispatchers|CreateEvent :!event Custom|CachedVolumeOverlapStart))
  (bind _outputdelegate_2 (EventDispatchers|CreateEvent :!event Custom|CachedVolumeOverlapEnd))
  (bind _cachedoverlapcomponent (Variables|Default|GetCachedOverlapComponent))
  (bind _fne_volumeoverlapcomponent (Variables|Default|GetFNE_VolumeOverlapComponent))
  ;[0 1 0.6126 0.25] Setting Event Bindings
  (Networking|SwitchHasAuthority
    (:Authority
      (Class|FNEVolumeOverlapComponent|UnBindfromonOtherActorBeginOverlap _cachedoverlapcomponent _outputdelegate)
      (Class|FNEVolumeOverlapComponent|UnBindfromonOtherActorEndOverlap _cachedoverlapcomponent _outputdelegate_2)
      (if (Variables|Default|GetHasExternalVolume)
        (Class|FNEVolumeOverlapComponent|UnBindfromonOtherActorBeginOverlap _fne_volumeoverlapcomponent _outputdelegate)
        (Class|FNEVolumeOverlapComponent|UnBindfromonOtherActorEndOverlap _fne_volumeoverlapcomponent _outputdelegate_2)))))

(event Custom|UpdateVolumeShape
  (bind _fne_volumeoverlapcomponent (Variables|Default|GetFNE_VolumeOverlapComponent))
  ;[0 1 0.6037 0.25] Update Shape and Size
  (bind _scale (CallFunction|UpdateVolumeSize (Variables|UserOptions|Advanced|GetVolumeShape) (Variables|UserOptions|Advanced|GetUse%20Custom%20Shape)))
  (Class|FNEVolumeOverlapComponent|UpdateOverlapShape _fne_volumeoverlapcomponent (Variables|UserOptions|Advanced|GetVolumeShape) (Variables|UserOptions|Advanced|GetUse%20Custom%20Shape) (CallFunction|GetCustomMesh))
  (CallFunction|UpdateVolumeRelativeLocation)
  (Class|FNEVolumeOverlapComponent|SetRelativeScale3DForBoundsComponent _fne_volumeoverlapcomponent _scale)
  ;[0 1 0.7707 0.25] Ensure proper material is applied when placing device.
  (if (not (Minigame|IsinPlayMode self :!class /Script/FortniteGame.MinigameBlueprintLibrary))
    (Class|FNEVolumeOverlapComponent|SetVolumeMaterial (Variables|Default|GetFNE_VolumeOverlapComponent) (select (Building|IsPreviewBuildingActor) (Variables|Default|GetPreviewMaterial) (Variables|Default|GetPlacedMaterial)) 0)
    (else
      (Class|FNEVolumeOverlapComponent|SetVolumeMaterial (Variables|Default|GetFNE_VolumeOverlapComponent) (select (Variables|UserOptions|Basic|GetVisibleInGame) (Variables|Default|GetPlacedMaterial) (Variables|Default|GetDefaultMaterial)) 0))))

(event Custom|RetryPlayerPawn
  (params (Pawn :type Pawn%20Object%20Reference))
  ;[0 1 0.6514 0.25] Retry if Player respawned in volume
  (bind _asfort_player_pawn_athena (Utilities|Casting|CastToFortPlayerPawnAthena Pawn))
  (Game|Player|UnbindEventfromOnPawnPossessed _asfort_player_pawn_athena (EventDispatchers|CreateEvent :!event Custom|RetryPlayerPawn))
  (CallFunction|FilterActorforEvent _asfort_player_pawn_athena))

(event Custom|FilterActorForEvent
  (params (ActorToFilter :type Actor%20Object%20Reference) (ExitedOverlap :type Boolean))
  (bind _actor_to_filter (Variables|Default|GetActor%20to%20Filter))
  (bind (_true _pawn) (CallFunction|IsCreature _actor_to_filter))
  (bind _returnvalue (Pawn|GetController _pawn))
  (bind _actor_to_filter_2 (Variables|Default|GetActor%20to%20Filter))
  (bind (_true_2 _pawn_2 _ai_bot_controller) (CallFunction|IsBot _actor_to_filter_2))
  (bind _actor_to_filter_3 (Variables|Default|GetActor%20to%20Filter))
  (bind _controller (CreativeAI|Controller|GetParticipantControllerfromActor _actor_to_filter_3 false))
  ;[0 1 0.872 0.25] Filter Overlapping Actor
  (Variables|Default|SetActor%20to%20Filter ActorToFilter)
  (Variables|Default|SetExited%20Overlap ExitedOverlap)
  (bind _asfort_vehicle_interface (Utilities|Casting|CastToFortVehicleInterface (Variables|Default|GetActor%20to%20Filter))
    (:then
      ;;[0.32 0 1 0.25] Vehicle Events
      (if (Variables|UserOptions|Advanced|GetVehicleEventsEnabled)
        (Utilities|IsValid (Vehicle|GetDriver _asfort_vehicle_interface)
          (:Is%20Valid
            (if (not (Variables|Default|GetExited%20Overlap))
              (GameplayEvent|BroadcastEvent (Variables|UserOptions|Events|GetOnEnter) 0 (Utilities|Struct|MakeEventVolumeDeviceEnter (Pawn|GetController (Vehicle|GetDriver _asfort_vehicle_interface))))
              (else
                (GameplayEvent|BroadcastEvent (Variables|UserOptions|Events|GetOnExit) 0 (Utilities|Struct|MakeEventVolumeDeviceExit (Pawn|GetController (Vehicle|GetDriver _asfort_vehicle_interface)))))))
          (:Is%20Not%20Valid
            _))
        (else
          _))
      (Utilities|IsValid _controller
        (:Is%20Valid
          (if (Variables|UserOptions|Basic|GetPlayerEventsEnabled)
            (Utilities|ValidTeamandClass (Variables|Default|GetActor%20to%20Filter)
              (:Valid
                (if (not (Variables|Default|GetExited%20Overlap))
                  (GameplayEvent|BroadcastEvent (Variables|UserOptions|Events|GetOnEnter) 0 (Utilities|Struct|MakeEventVolumeDeviceEnter _controller))
                  (else
                    (GameplayEvent|BroadcastEvent (Variables|UserOptions|Events|GetOnExit) 0 (Utilities|Struct|MakeEventVolumeDeviceExit _controller))))))
            (else
              _)))
        (:Is%20Not%20Valid
          _))
      ;;[0.32 0 1 0.25] Creature Events
      (if _true
        (if (Variables|UserOptions|Advanced|GetCreatureAndWildlifeEventsEnabled)
          (if (not (Variables|Default|GetExited%20Overlap))
            (GameplayEvent|BroadcastEvent (Variables|UserOptions|Events|GetOnEnter) 0 (Utilities|Struct|MakeEventVolumeDeviceEnter _returnvalue))
            (else
              (GameplayEvent|BroadcastEvent (Variables|UserOptions|Events|GetOnExit) 0 (Utilities|Struct|MakeEventVolumeDeviceExit _returnvalue))))
          (else
            _))
        (else
          _))
      (bind _isguard (CreativeAI|NPCType|IsGuardActor _actor_to_filter_2))
      (if _isguard
        (if (Variables|UserOptions|Advanced|GetGuardEventsEnabled)
          (Utilities|ValidTeamandClass (Variables|Default|GetActor%20to%20Filter)
            (:Valid
              (if (not (Variables|Default|GetExited%20Overlap))
                (GameplayEvent|BroadcastEvent (Variables|UserOptions|Events|GetOnEnter) 0 (Utilities|Struct|MakeEventVolumeDeviceEnter _ai_bot_controller))
                (else
                  (GameplayEvent|BroadcastEvent (Variables|UserOptions|Events|GetOnExit) 0 (Utilities|Struct|MakeEventVolumeDeviceExit _ai_bot_controller)))))))
        (else
          (bind _asbuilding_prop (Utilities|Casting|CastToBuildingProp (Variables|Default|GetActor%20to%20Filter))
            (:then
              (if (and (Utilities|IsValid (Actor|GetComponentbyClass _asbuilding_prop "/Script/FortniteGame.FortPhysicsComponent") :!class /Script/Engine.KismetSystemLibrary) (Variables|UserOptions|Basic|GetPhysicsEventsEnabled))
                ;;[0.1118 0.15 0.002344 0.5] Register Prop Event Analytics for Newton - Instigated Events per Minute
                ;;[1 0 0 0.5] NOTE: When props eventually can track the last instigator, that should be the instigator passed and called only once - Currently the event is tracked globally, not linked to a particular player, but using them as means to track the event - Check FORT-915953 and FORT-874255 and FORT-922603
                (CallFunction|DeviceTelemetryRegisteronPropOverlapEvent _asbuilding_prop (Variables|Default|GetExited%20Overlap))
                (if (not (Variables|Default|GetExited%20Overlap))
                  (GameplayEvent|BroadcastEvent (Variables|UserOptions|Events|GetOnPhysicsEnter) 0 (Utilities|Struct|MakeEventVolumeDevicePhysicsEnter _asbuilding_prop))
                  (else
                    (GameplayEvent|BroadcastEvent (Variables|UserOptions|Events|GetOnPhysicsExit) 0 (Utilities|Struct|MakeEventVolumeDevicePhysicsExit _asbuilding_prop))))
                (else
                  _)))
            (:CastFailed
              _))
          ;;[0 1 0.8149 0.25] Bind to possessed event if player respawned in volume
          (if (Variables|UserOptions|Basic|GetPlayerEventsEnabled)
            (bind _asfort_player_pawn_athena (Utilities|Casting|CastToFortPlayerPawnAthena (Variables|Default|GetActor%20to%20Filter))
              (:then
                (Game|Player|BindEventtoOnPawnPossessed _asfort_player_pawn_athena (EventDispatchers|CreateEvent :!event Custom|RetryPlayerPawn)))
              (:CastFailed
                _))
            (else
              _))
          (CallFunction|NormalorShippingLog (Utilities|String|Append (Utilities|GetDisplayName self) "Filter Actor For Event: Unidentified Actor, no event sent.")))))
    (:CastFailed
      _)))

(event ChildFunctions|EventPlayModeChanged
  (params (Minigame :type Fort%20Minigame%20Object%20Reference) (bIsInPlayMode :type Boolean))
  ;[0 1 0.6162 0.25] Update Volume Shape on Playmode Change to set material correctly.
  (|CallParentFunction Minigame bIsInPlayMode)
  (CallFunction|UpdateVolumeShape))

(event Custom|DeviceTelemetry_RegisterOnPropOverlapEvent
  (params (PropInstigator :type Building%20Prop%20Object%20Reference) (ExitedOverlap :type Boolean))
  ;[0 1 0.872 0.25] Filter Overlapping Actor
  ;;[0.1118 0.15 0.002344 0.5] Register Prop Event Analytics for Newton - Instigated Events per Minute
  ;;[1 0 0 0.5] NOTE: When props eventually can track the last instigator, that should be the instigator passed and called only once - Currently the event is tracked globally, not linked to a particular player, but using them as means to track the event - Check FORT-915953 and FORT-874255 and FORT-922603
  (Devices|Analytics|DeviceTelemetryRegisteronPropOverlapEvent PropInstigator ExitedOverlap))

(fn CheckVolumeToCache
  (bind _fne_volumeoverlapcomponent (Variables|Default|GetFNE_VolumeOverlapComponent))
  ;[0 1 0.7064 0.25] Cache the volume component from this device, unless an external volume is set up.
  (bind _externalvolume (Variables|UserOptions|Advanced|GetExternalVolume)
    (:then
      (CallFunction|NormalorShippingLog "Volume Device: External Volume is valid!")
      (if (== (Variables|UserOptions|Advanced|GetExternalVolume _externalvolume) 0)
        (CallFunction|NormalorShippingLog "Volume Device: External Volume is not hooked up to another volume!")
        (Utilities|IsValid (Variables|Default|GetFNE_VolumeOverlapComponent _externalvolume)
          (:Is%20Valid
            (CallFunction|NormalorShippingLog "Volume Device: External Volume has a valid FNE Volume component")
            (Variables|Default|SetCachedOverlapComponent (Variables|Default|GetFNE_VolumeOverlapComponent _externalvolume))
            (Variables|Default|SetHasExternalVolume true))
          (:Is%20Not%20Valid
            (CallFunction|NormalorShippingLog "Volume Device: External Volume has NO valid FNE Volume component")))
        (else
          (CallFunction|NormalorShippingLog "Volume Device: External Volume IS hooked up to another volume!")
          (Utilities|IsValid _fne_volumeoverlapcomponent
            (:Is%20Valid
              _))))
      (Variables|Default|SetCachedOverlapComponent _fne_volumeoverlapcomponent)
      (Variables|Default|SetHasExternalVolume false))
    (:else
      _)))

(fn GetCustomMesh
  (return (Variables|UserOptions|Advanced|GetCustomVolumeMesh)))

(fn GetCollisionComponent
  (return (Variables|Default|GetFNE_VolumeOverlapComponent)))

(fn GetOccupiers
  ;[0.9594 0 1 0.5] Returns any agents in this volume.
  (bind _output_get (LocalVariables|SetPawns (Class|FNEVolumeOverlapComponent|GetAllPlayerPawns (Variables|Default|GetFNE_VolumeOverlapComponent))))
  (for _array_element _output_get
    (Utilities|IsValid (Fortnite|GetControllerfromActor _array_element)
      (:Is%20Valid
        (Utilities|Array|Add (LocalVariables|GetOccupyingControllers) (Fortnite|GetControllerfromActor _array_element)))))
  (return (LocalVariables|GetOccupyingControllers)))

(fn InitializeCreativeIcon
  ;[0 1 0.7494 0.25] Resize Creative icon when scaling via Transform. We want to keep the Icon scale consistant.
  (Transformation|SetWorldScale3D (Variables|Default|GetCreativeIcon) "1.000000,1.000000,1.000000"))

(fn IsBot
  (params (Actor :type Actor%20Object%20Reference))
  (bind _returnvalue (AI|GetAIController Actor))
  ;[0 1 0.8575 0.25] Is this Actor a Bot?
  (Utilities|IsValid _returnvalue
    (:Is%20Valid
      (bind _asfort_athena_aibot_controller (Utilities|Casting|CastToFortAthenaAIBotController _returnvalue)
        (:then
          (return true (Variables|/Script/FortniteAI.FortAthenaAIBotController|GetPlayerBotPawn _asfort_athena_aibot_controller) _asfort_athena_aibot_controller))
        (:CastFailed
          _)))
    (:Is%20Not%20Valid
      _))
  (return false 0 0))

(fn IsCreature
  (params (Actor :type Actor%20Object%20Reference))
  ;[0 1 0.7885 0.25] Is this Actor a Creature?
  (Utilities|Casting|CastToFortPlayerPawn Actor
    (:then
      (return false 0))
    (:CastFailed
      (bind _asfort_pawn (Utilities|Casting|CastToFortPawn Actor)
        (:then
          (return true _asfort_pawn))
        (:CastFailed
          (return false 0))))))

(fn IsActorInVolume
  (params (Actor :type Actor%20Object%20Reference))
  ;[0 1 0.9671 0.25] TODO: Replace with Tracked Actors on the volume component once implemented.
  (return (or (and (Variables|Default|GetHasExternalVolume) (Utilities|Array|ContainsItem (Class|FNEVolumeOverlapComponent|GetAllTrackedActors (Variables|Default|GetFNE_VolumeOverlapComponent)) (Pawn|GetControlledPawn (Fortnite|GetControllerfromActor Actor)))) (Utilities|Array|ContainsItem (Class|FNEVolumeOverlapComponent|GetAllTrackedActors (Variables|Default|GetCachedOverlapComponent)) (Pawn|GetControlledPawn (Fortnite|GetControllerfromActor Actor))))))

(fn IsAgentInVolume
  (params (Controller :type Controller%20Object%20Reference))
  ;[0 1 0.9671 0.25] Verse - Returns if the Agent is in the Volume
  (return (or (and (Variables|Default|GetHasExternalVolume) (Utilities|Array|ContainsItem (Class|FNEVolumeOverlapComponent|GetAllTrackedActors (Variables|Default|GetFNE_VolumeOverlapComponent)) (Pawn|GetControlledPawn Controller))) (Utilities|Array|ContainsItem (Class|FNEVolumeOverlapComponent|GetAllTrackedActors (Variables|Default|GetCachedOverlapComponent)) (Pawn|GetControlledPawn Controller)))))

(fn NormalOrShippingLog
  (params (InString :type String))
  (if (Variables|Default|GetbUseShippingLogs)
    (Fortnite|FortShippingLog self InString)
    (else
      (Development|PrintString InString false))))

(fn RevalidateActor)

(fn UpdateVolumeCollision
  (params (IsCollisionEnabled :type Boolean))
  (Collision|SetActorEnableCollision (Variables|Default|GetCachedPlayspaceVolume) IsCollisionEnabled))

(fn UpdateVisibilityComponent
  (Initialization|InitializeWithSceneComponents (Variables|Default|GetVisibleInGameComponent) (Utilities|Array|MakeArray (Variables|Debug|GetStaticMeshComponent) (Variables|Default|GetFNE_VolumeOverlapComponent)) (Utilities|Map|MakeMap) (select (Variables|UserOptions|Basic|GetVisibleInGame) "NewEnumerator3" "NewEnumerator1")))

(fn UpdateVolumeRelativeLocation
  (Transformation|SetRelativeLocation (Variables|Default|GetFNE_VolumeOverlapComponent) (Math|Vector|MakeVector 0.0 0.0 (Math|Float|SelectFloat 0.0 -192.0 (Utilities|Enum|Equal(Enum) (Variables|UserOptions|Advanced|GetVolumeShape) "SphereVolume"))) false false))

(fn UpdateVolumeSize
  (params (Volume%20Shape :type EFNEVolumeShapeTypeEnum%20Enum :default "BoxVolume") (Use%20Custom%20Shape :type Boolean))
  (if Use%20Custom%20Shape
    (return (Math|Vector|MakeVector 1.0 1.0 1.0))
    (else
      (switch Utilities|FlowControl|Switch|SwitchonEFNEVolumeShapeTypeEnum Volume%20Shape
        (:BoxVolume
          (return (Math|Vector|MakeVector (Variables|UserOptions|Basic|GetVolumeWidth) (Variables|UserOptions|Basic|GetVolumeDepth) (Variables|UserOptions|Basic|GetVolumeHeight))))
        (:CylinderVolume
          (return (Math|Vector|MakeVector (Variables|UserOptions|Basic|GetVolumeRadius) (Variables|UserOptions|Basic|GetVolumeRadius) (Variables|UserOptions|Basic|GetVolumeHeight))))
        (:SphereVolume
          (return (Math|Vector|MakeVector (* (Variables|UserOptions|Basic|GetVolumeRadius) (Variables|Default|GetM%20to%20Scale)) (* (Variables|UserOptions|Basic|GetVolumeRadius) (Variables|Default|GetM%20to%20Scale)) (* (Variables|UserOptions|Basic|GetVolumeRadius) (Variables|Default|GetM%20to%20Scale)))))
        (:CustomVolume
          (return "1.000000,1.000000,1.000000"))))))

(fn ConstructionScript
  (bind _fne_volumeoverlapcomponent (Variables|Default|GetFNE_VolumeOverlapComponent))
  (|CallParentFunction)
  ;[1 1 1] Force Collision Object Type and Enabled to match with Component default
  (Collision|SetCollisionObjectType _fne_volumeoverlapcomponent "ECC_WorldDynamic")
  (Collision|SetCollisionEnabled _fne_volumeoverlapcomponent "QueryOnly")
  (if (not (Editor|InEditor))
    (Rendering|Material|SetMaterial (Variables|Default|GetFNE_VolumeOverlapComponent) 0 0) ; Attempting to fix FORT-683185, reapplying invisible material\nearly in the process to reset any material that was saved on\nthe component instance.
  ))
