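; VolumeDeviceFNE_Volume is a Blueprint for the volume of a volume device that cleans up its playspace when destroyed.
; On EndPlay, it checks whether the reason is Destroyed and, if so, passes the result of GetPlayspace to PlayspaceLibrary's DestroyPlayspace.
; Other end-play reasons leave the playspace untouched, and the construction script is empty.
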(event EventEndPlay
  (params (EndPlayReason :type EEndPlayReason%20Enum :default "Destroyed"))
  (if (Utilities|Enum|Equal(Enum) EndPlayReason "Destroyed")
    (Class|PlayspaceLibrary|DestroyPlayspace (CallFunction|GetPlayspace))))

(fn ConstructionScript)
